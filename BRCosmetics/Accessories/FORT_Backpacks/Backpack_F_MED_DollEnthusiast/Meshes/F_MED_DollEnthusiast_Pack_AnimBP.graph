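# Animation Blueprint for the DollEnthusiast backpack, adding secondary physics and a body-fit adjustment.
# AnimDynamics chains simulate charm and handle bones in front end; the other branch also simulates fifteen pack-relative strands, blended by bIsInFrontEnd.
# That non-front-end branch scales handle simulation alpha with frame rate, computed each update.
# On customization load, a wildcard mesh-name check enables a slight pack-bone offset when the wearer isn't DollEnthusiast.
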AnimBlueprint /BRCosmetics/Accessories/FORT_Backpacks/Backpack_F_MED_DollEnthusiast/Meshes/F_MED_DollEnthusiast_Pack_AnimBP
target_skeleton=&/F_MED_DollEnthusiast_Pack_Skeleton
graphs [
 AnimGraph {nodes [
   n1 = Anim Dynamics - Bone: dyn_charm_01(ComponentPose=Local To Component(LocalPose=Use cached pose 'InputPose'()), Alpha=0.8, ExternalForce="0,0,0") {
    angularDampingOverride=0.999
    angularSpringConstant=300
    bAngularSpring=true
    bOverrideAngularDamping=true
    bUseGravityOverride=true
    simulationSpace=World
    alphaBoolBlend {blendOption=Linear}
    boundBone {boneName=dyn_charm_01}
    componentAppliedLinearAccClamp=<2000 2000 2000>
    componentLinearAccScale=<0.2 0.2 0.2>
    componentLinearVelScale=<0.2 0.2 0.2>
    gravityOverride {z=-400}
    physicsBodyDefinitions [{
      boundBone {boneName=dyn_charm_01}
      boxExtents=<12 12 12>
      constraintSetup {angularLimitsMax=<0 5 8>, angularLimitsMin=<0 -8 -8>, angularTarget=<0 0 0>}
      localJointOffset=<2 0 0>
     }]
   }
   n2 = Anim Dynamics - Bone: dyn_charm_02(ComponentPose=@n1, Alpha=0.8, ExternalForce="0,0,0") {
    angularDampingOverride=0.999
    bOverrideAngularDamping=true
    bUseGravityOverride=true
    simulationSpace=World
    alphaBoolBlend {blendOption=Linear}
    boundBone {boneName=dyn_charm_02}
    componentAppliedLinearAccClamp=<2000 2000 2000>
    componentLinearAccScale=<0.2 0.2 0.2>
    componentLinearVelScale=<0.2 0.2 0.2>
    gravityOverride {z=-500}
    physicsBodyDefinitions [{
      boundBone {boneName=dyn_charm_02}
      boxExtents=<12 12 12>
      constraintSetup {angularLimitsMax=<10 5 10>, angularLimitsMin=<-10 -7 -15>, angularTarget=<0 0 0>}
      localJointOffset=<2 0 0>
     }]
   }
   n3 = Anim Dynamics - Bone: dyn_charm_03(ComponentPose=@n2, Alpha=0.8, ExternalForce="0,0,0") {
    angularDampingOverride=0.999
    angularSpringConstant=300
    bAngularSpring=true
    bOverrideAngularDamping=true
    bUseGravityOverride=true
    simulationSpace=World
    alphaBoolBlend {blendOption=Linear}
    boundBone {boneName=dyn_charm_03}
    componentAppliedLinearAccClamp=<2000 2000 2000>
    componentLinearAccScale=<0.2 0.2 0.2>
    componentLinearVelScale=<0.2 0.2 0.2>
    gravityOverride {z=-600}
    physicsBodyDefinitions [{
      boundBone {boneName=dyn_charm_03}
      boxExtents=<12 12 12>
      constraintSetup {angularLimitsMax=<1 5 15>, angularLimitsMin=<-1 -7 -15>, angularTarget=<0 0 0>}
      localJointOffset=<2 0 0>
     }]
   }
   n4 = Anim Dynamics - Bone: dyn_handle_01(ComponentPose=@n3, Alpha=1, ExternalForce="0,0,0") {
    angularDampingOverride=0.9999
    angularSpringConstant=300
    bAngularSpring=true
    bOverrideAngularDamping=true
    bUseGravityOverride=true
    simulationSpace=World
    alphaBoolBlend {blendOption=Linear}
    boundBone {boneName=dyn_handle_01}
    componentAppliedLinearAccClamp=<2000 2000 2000>
    componentLinearAccScale=<0.2 0.2 0.2>
    componentLinearVelScale=<0.2 0.2 0.2>
    gravityOverride {z=-600}
    physicsBodyDefinitions [{
      boundBone {boneName=dyn_handle_01}
      boxExtents=<20 20 20>
      constraintSetup {angularLimitsMax=<0 20 0>, angularLimitsMin=<0 -70 0>, angularTarget=<0 0 0>}
      localJointOffset=<2 0 0>
     }]
   }
   n5 = Anim Dynamics - Bone: dyn_handle_02(ComponentPose=@n4, Alpha=1, ExternalForce="0,0,0") {
    bUseGravityOverride=true
    simulationSpace=World
    alphaBoolBlend {blendOption=Linear}
    boundBone {boneName=dyn_handle_02}
    componentAppliedLinearAccClamp=<600 1000 400>
    componentLinearAccScale=<0.2 0.2 0.2>
    componentLinearVelScale=<0.2 0.2 0.2>
    gravityOverride {z=-400}
    physicsBodyDefinitions [{
      boundBone {boneName=dyn_handle_02}
      boxExtents=<9 9 9>
      constraintSetup {angularLimitsMax=<0 20 0>, angularLimitsMin=<0 -60 0>, angularTarget=<0 0 0>}
      localJointOffset=<1.5 0 0>
     }]
   }
   n6 = Component To Local(ComponentPose=@n5)
   n7 = Anim Dynamics - Bone: dyn_charm_01(ComponentPose=Local To Component(LocalPose=Use cached pose 'InputPose'()), Alpha=0.8, ExternalForce="0,0,0") {
    angularDampingOverride=0.999
    angularSpringConstant=300
    bAngularSpring=true
    bOverrideAngularDamping=true
    bUseGravityOverride=true
    simulationSpace=World
    alphaBoolBlend {blendOption=Linear}
    boundBone {boneName=dyn_charm_01}
    componentAppliedLinearAccClamp=<2000 2000 2000>
    componentLinearAccScale=<0.2 0.2 0.2>
    componentLinearVelScale=<0.2 0.2 0.2>
    gravityOverride {z=-400}
    physicsBodyDefinitions [{
      boundBone {boneName=dyn_charm_01}
      boxExtents=<12 12 12>
      constraintSetup {angularLimitsMax=<0 5 8>, angularLimitsMin=<0 -8 -8>, angularTarget=<0 0 0>}
      localJointOffset=<2 0 0>
     }]
   }
   n8 = Anim Dynamics - Bone: dyn_charm_02(ComponentPose=@n7, Alpha=0.8, ExternalForce="0,0,0") {
    angularDampingOverride=0.999
    bOverrideAngularDamping=true
    bUseGravityOverride=true
    simulationSpace=World
    alphaBoolBlend {blendOption=Linear}
    boundBone {boneName=dyn_charm_02}
    componentAppliedLinearAccClamp=<2000 2000 2000>
    componentLinearAccScale=<0.2 0.2 0.2>
    componentLinearVelScale=<0.2 0.2 0.2>
    gravityOverride {z=-500}
    physicsBodyDefinitions [{
      boundBone {boneName=dyn_charm_02}
      boxExtents=<12 12 12>
      constraintSetup {angularLimitsMax=<10 5 10>, angularLimitsMin=<-10 -7 -15>, angularTarget=<0 0 0>}
      localJointOffset=<2 0 0>
     }]
   }
   n9 = Anim Dynamics - Bone: dyn_charm_03(ComponentPose=@n8, Alpha=0.8, ExternalForce="0,0,0") {
    angularDampingOverride=0.999
    angularSpringConstant=300
    bAngularSpring=true
    bOverrideAngularDamping=true
    bUseGravityOverride=true
    simulationSpace=World
    alphaBoolBlend {blendOption=Linear}
    boundBone {boneName=dyn_charm_03}
    componentAppliedLinearAccClamp=<2000 2000 2000>
    componentLinearAccScale=<0.2 0.2 0.2>
    componentLinearVelScale=<0.2 0.2 0.2>
    gravityOverride {z=-600}
    physicsBodyDefinitions [{
      boundBone {boneName=dyn_charm_03}
      boxExtents=<12 12 12>
      constraintSetup {angularLimitsMax=<1 5 15>, angularLimitsMin=<-1 -7 -15>, angularTarget=<0 0 0>}
      localJointOffset=<2 0 0>
     }]
   }
   n10 = Anim Dynamics - Bone: dyn_handle_01(ComponentPose=@n9, Alpha=Get FPSAlpha FE() -> (FPSAlpha FE:double), ExternalForce="0,0,0") {
    angularDampingOverride=0.9999
    angularSpringConstant=300
    bAngularSpring=true
    bOverrideAngularDamping=true
    bUseGravityOverride=true
    simulationSpace=World
    alphaBoolBlend {blendOption=Linear}
    boundBone {boneName=dyn_handle_01}
    componentAppliedLinearAccClamp=<2000 2000 2000>
    componentLinearAccScale=<0.2 0.2 0.2>
    componentLinearVelScale=<0.2 0.2 0.2>
    gravityOverride {z=-600}
    physicsBodyDefinitions [{
      boundBone {boneName=dyn_handle_01}
      boxExtents=<20 20 20>
      constraintSetup {angularLimitsMax=<0 20 0>, angularLimitsMin=<0 -70 0>, angularTarget=<0 0 0>}
      localJointOffset=<2 0 0>
     }]
   }
   n11 = Anim Dynamics - Bone: dyn_handle_02(ComponentPose=@n10, Alpha=Get FPSAlpha FE(), ExternalForce="0,0,0") {
    bUseGravityOverride=true
    simulationSpace=World
    alphaBoolBlend {blendOption=Linear}
    boundBone {boneName=dyn_handle_02}
    componentAppliedLinearAccClamp=<600 1000 400>
    componentLinearAccScale=<0.2 0.2 0.2>
    componentLinearVelScale=<0.2 0.2 0.2>
    gravityOverride {z=-400}
    physicsBodyDefinitions [{
      boundBone {boneName=dyn_handle_02}
      boxExtents=<9 9 9>
      constraintSetup {angularLimitsMax=<0 20 0>, angularLimitsMin=<0 -60 0>, angularTarget=<0 0 0>}
      localJointOffset=<1.5 0 0>
     }]
   }
   n12 = Anim Dynamics - Bone: dyn_strand_01!(ComponentPose=@n11, Alpha=1) -> (Pose) {
    angularSpringConstant=0.999
    bAngularSpring=true
    bUseGravityOverride=true
    bUseSphericalLimits=true
    simulationSpace=BoneRelative
    alphaBoolBlend {blendOption=Linear}
    boundBone {boneName=dyn_strand_01}
    componentAppliedLinearAccClamp=<500 500 500>
    componentLinearAccScale=<0.5 0.5 1>
    componentLinearVelScale=<0.5 0.5 1>
    gravityOverride {z=-200}
    physicsBodyDefinitions [{
      boundBone {boneName=dyn_strand_01}
      boxExtents=<5 10 10>
      constraintSetup {angularLimitsMax=<30 30 30>, angularLimitsMin=<-30 -30 -30>, angularTarget=<0 0 0>}
      localJointOffset=<1 0 0>
     }]
    relativeSpaceBone {boneName=pack}
    sphericalLimits [{limitRadius=0, drivingBone {boneName=None}, sphereLocalOffset=<0 0 0>}]
   }
   n13 = Anim Dynamics - Bone: dyn_strand_02!(ComponentPose=@n12, Alpha=1) -> (Pose) {
    angularDampingOverride=0.9
    angularSpringConstant=0.999
    bOverrideAngularDamping=true
    bUseGravityOverride=true
    bUseSphericalLimits=true
    simulationSpace=BoneRelative
    alphaBoolBlend {blendOption=Linear}
    boundBone {boneName=dyn_strand_02}
    componentAppliedLinearAccClamp=<750 750 750>
    componentLinearAccScale=<0.5 0.5 1>
    componentLinearVelScale=<0.5 0.5 1>
    gravityOverride {z=-500}
    physicsBodyDefinitions [{
      boundBone {boneName=dyn_strand_02}
      boxExtents=<10 15 15>
      constraintSetup {angularLimitsMax=<30 30 30>, angularLimitsMin=<-30 -30 -30>, angularTarget=<0 0 0>}
      localJointOffset=<0.75 0 0>
     }]
    relativeSpaceBone {boneName=pack}
   }
   n14 = Anim Dynamics - Bone: dyn_strand_03!(ComponentPose=@n13, Alpha=1) -> (Pose) {
    angularSpringConstant=0.999
    bAngularSpring=true
    bUseGravityOverride=true
    bUseSphericalLimits=true
    simulationSpace=BoneRelative
    alphaBoolBlend {blendOption=Linear}
    boundBone {boneName=dyn_strand_03}
    componentAppliedLinearAccClamp=<500 500 500>
    componentLinearAccScale=<0.5 0.5 1>
    componentLinearVelScale=<0.5 0.5 1>
    gravityOverride {z=-200}
    physicsBodyDefinitions [{
      boundBone {boneName=dyn_strand_03}
      boxExtents=<5 10 10>
      constraintSetup {angularLimitsMax=<30 30 30>, angularLimitsMin=<-30 -30 -30>, angularTarget=<0 0 0>}
      localJointOffset=<1 0 0>
     }]
    relativeSpaceBone {boneName=pack}
    sphericalLimits [{limitRadius=0, drivingBone {boneName=None}, sphereLocalOffset=<0 0 0>}]
   }
   n15 = Anim Dynamics - Bone: dyn_strand_04!(ComponentPose=@n14, Alpha=1) -> (Pose) {
    angularSpringConstant=0.999
    bAngularSpring=true
    bUseGravityOverride=true
    bUseSphericalLimits=true
    simulationSpace=BoneRelative
    alphaBoolBlend {blendOption=Linear}
    boundBone {boneName=dyn_strand_04}
    componentAppliedLinearAccClamp=<500 500 500>
    componentLinearAccScale=<0.5 0.5 1>
    componentLinearVelScale=<0.5 0.5 1>
    gravityOverride {z=-200}
    physicsBodyDefinitions [{
      boundBone {boneName=dyn_strand_04}
      boxExtents=<10 15 15>
      constraintSetup {angularLimitsMax=<30 30 30>, angularLimitsMin=<-30 -30 -30>, angularTarget=<0 0 0>}
      localJointOffset=<1 0 0>
     }]
    relativeSpaceBone {boneName=pack}
   }
   n16 = Anim Dynamics - Bone: dyn_strand_05!(ComponentPose=@n15, Alpha=1) -> (Pose) {
    angularSpringConstant=0.999
    bAngularSpring=true
    bUseGravityOverride=true
    bUseSphericalLimits=true
    simulationSpace=BoneRelative
    alphaBoolBlend {blendOption=Linear}
    boundBone {boneName=dyn_strand_05}
    componentAppliedLinearAccClamp=<500 500 500>
    componentLinearAccScale=<0.5 0.5 1>
    componentLinearVelScale=<0.5 0.5 1>
    gravityOverride {z=-200}
    physicsBodyDefinitions [{
      boundBone {boneName=dyn_strand_05}
      boxExtents=<5 10 10>
      constraintSetup {angularLimitsMax=<30 30 30>, angularLimitsMin=<-30 -30 -30>, angularTarget=<0 0 0>}
      localJointOffset=<1 0 0>
     }]
    relativeSpaceBone {boneName=pack}
    sphericalLimits [{limitRadius=0, drivingBone {boneName=None}, sphereLocalOffset=<0 0 0>}]
   }
   n17 = Anim Dynamics - Bone: dyn_strand_06!(ComponentPose=@n16, Alpha=1) -> (Pose) {
    angularSpringConstant=0.999
    bAngularSpring=true
    bUseGravityOverride=true
    bUseSphericalLimits=true
    simulationSpace=BoneRelative
    alphaBoolBlend {blendOption=Linear}
    boundBone {boneName=dyn_strand_06}
    componentAppliedLinearAccClamp=<500 500 500>
    componentLinearAccScale=<0.5 0.5 1>
    componentLinearVelScale=<0.5 0.5 1>
    gravityOverride {z=-200}
    physicsBodyDefinitions [{
      boundBone {boneName=dyn_strand_06}
      boxExtents=<5 10 10>
      constraintSetup {angularLimitsMax=<30 30 30>, angularLimitsMin=<-30 -30 -30>, angularTarget=<0 0 0>}
      localJointOffset=<1 0 0>
     }]
    relativeSpaceBone {boneName=pack}
    sphericalLimits [{limitRadius=0, drivingBone {boneName=None}, sphereLocalOffset=<0 0 0>}]
   }
   n18 = Anim Dynamics - Bone: dyn_strand_07!(ComponentPose=@n17, Alpha=1) -> (Pose) {
    angularSpringConstant=0.999
    bAngularSpring=true
    bUseGravityOverride=true
    bUseSphericalLimits=true
    simulationSpace=BoneRelative
    alphaBoolBlend {blendOption=Linear}
    boundBone {boneName=dyn_strand_07}
    componentAppliedLinearAccClamp=<500 500 500>
    componentLinearAccScale=<0.5 0.5 1>
    componentLinearVelScale=<0.5 0.5 1>
    gravityOverride {z=-200}
    physicsBodyDefinitions [{
      boundBone {boneName=dyn_strand_07}
      boxExtents=<10 15 15>
      constraintSetup {angularLimitsMax=<30 30 30>, angularLimitsMin=<-30 -30 -30>, angularTarget=<0 0 0>}
      localJointOffset=<0.75 0 0>
     }]
    relativeSpaceBone {boneName=pack}
   }
   n19 = Anim Dynamics - Bone: dyn_strand_08!(ComponentPose=@n18, Alpha=1) -> (Pose) {
    angularSpringConstant=0.999
    bAngularSpring=true
    bUseGravityOverride=true
    bUseSphericalLimits=true
    simulationSpace=BoneRelative
    alphaBoolBlend {blendOption=Linear}
    boundBone {boneName=dyn_strand_08}
    componentAppliedLinearAccClamp=<500 500 500>
    componentLinearAccScale=<0.5 0.5 1>
    componentLinearVelScale=<0.5 0.5 1>
    gravityOverride {z=-200}
    physicsBodyDefinitions [{
      boundBone {boneName=dyn_strand_08}
      boxExtents=<5 10 10>
      constraintSetup {angularLimitsMax=<30 30 30>, angularLimitsMin=<-30 -30 -30>, angularTarget=<0 0 0>}
      localJointOffset=<1 0 0>
     }]
    relativeSpaceBone {boneName=pack}
    sphericalLimits [{limitRadius=0, drivingBone {boneName=None}, sphereLocalOffset=<0 0 0>}]
   }
   n20 = Anim Dynamics - Bone: dyn_strand_09!(ComponentPose=@n19, Alpha=1) -> (Pose) {
    angularSpringConstant=0.999
    bAngularSpring=true
    bUseGravityOverride=true
    bUseSphericalLimits=true
    simulationSpace=BoneRelative
    alphaBoolBlend {blendOption=Linear}
    boundBone {boneName=dyn_strand_09}
    componentAppliedLinearAccClamp=<500 500 500>
    componentLinearAccScale=<0.5 0.5 1>
    componentLinearVelScale=<0.5 0.5 1>
    gravityOverride {z=-200}
    physicsBodyDefinitions [{
      boundBone {boneName=dyn_strand_09}
      boxExtents=<10 15 15>
      constraintSetup {angularLimitsMax=<30 30 30>, angularLimitsMin=<-30 -30 -30>, angularTarget=<0 0 0>}
      localJointOffset=<1 0 0>
     }]
    relativeSpaceBone {boneName=pack}
   }
   n21 = Anim Dynamics - Bone: dyn_strand_10!(ComponentPose=@n20, Alpha=1) -> (Pose) {
    angularSpringConstant=0.999
    bAngularSpring=true
    bUseGravityOverride=true
    bUseSphericalLimits=true
    simulationSpace=BoneRelative
    alphaBoolBlend {blendOption=Linear}
    boundBone {boneName=dyn_strand_10}
    componentAppliedLinearAccClamp=<500 500 500>
    componentLinearAccScale=<0.5 0.5 1>
    componentLinearVelScale=<0.5 0.5 1>
    gravityOverride {z=-200}
    physicsBodyDefinitions [{
      boundBone {boneName=dyn_strand_10}
      boxExtents=<5 10 10>
      constraintSetup {angularLimitsMax=<0 0 50>, angularLimitsMin=<0 -75 -50>, angularTarget=<0 0 0>}
      localJointOffset=<1 0 0>
     }]
    relativeSpaceBone {boneName=pack}
    sphericalLimits [{limitRadius=0, drivingBone {boneName=None}, sphereLocalOffset=<0 0 0>}]
   }
   n22 = Anim Dynamics - Bone: dyn_strand_11!(ComponentPose=@n21, Alpha=1) -> (Pose) {
    angularSpringConstant=0.999
    bAngularSpring=true
    bUseGravityOverride=true
    bUseSphericalLimits=true
    simulationSpace=BoneRelative
    alphaBoolBlend {blendOption=Linear}
    boundBone {boneName=dyn_strand_11}
    componentAppliedLinearAccClamp=<500 500 500>
    componentLinearAccScale=<0.5 0.5 1>
    componentLinearVelScale=<0.5 0.5 1>
    gravityOverride {z=-200}
    physicsBodyDefinitions [{
      boundBone {boneName=dyn_strand_11}
      boxExtents=<10 15 15>
      constraintSetup {angularLimitsMax=<0 0 50>, angularLimitsMin=<0 -75 -50>, angularTarget=<0 0 0>}
      localJointOffset=<0.75 0 0>
     }]
    relativeSpaceBone {boneName=pack}
   }
   n23 = Anim Dynamics - Bone: dyn_strand_12!(ComponentPose=@n22, Alpha=1) -> (Pose) {
    angularSpringConstant=0.999
    bAngularSpring=true
    bUseGravityOverride=true
    bUseSphericalLimits=true
    simulationSpace=BoneRelative
    alphaBoolBlend {blendOption=Linear}
    boundBone {boneName=dyn_strand_12}
    componentAppliedLinearAccClamp=<500 500 500>
    componentLinearAccScale=<0.5 0.5 1>
    componentLinearVelScale=<0.5 0.5 1>
    gravityOverride {z=-200}
    physicsBodyDefinitions [{
      boundBone {boneName=dyn_strand_12}
      boxExtents=<5 10 10>
      constraintSetup {angularLimitsMax=<0 0 50>, angularLimitsMin=<0 -75 -50>, angularTarget=<0 0 0>}
      localJointOffset=<1 0 0>
     }]
    relativeSpaceBone {boneName=pack}
    sphericalLimits [{limitRadius=0, drivingBone {boneName=None}, sphereLocalOffset=<0 0 0>}]
   }
   n24 = Anim Dynamics - Bone: dyn_strand_13!(ComponentPose=@n23, Alpha=1) -> (Pose) {
    angularSpringConstant=0.999
    bAngularSpring=true
    bUseGravityOverride=true
    bUseSphericalLimits=true
    simulationSpace=BoneRelative
    alphaBoolBlend {blendOption=Linear}
    boundBone {boneName=dyn_strand_13}
    componentAppliedLinearAccClamp=<500 500 500>
    componentLinearAccScale=<0.5 0.5 1>
    componentLinearVelScale=<0.5 0.5 1>
    gravityOverride {z=-200}
    physicsBodyDefinitions [{
      boundBone {boneName=dyn_strand_13}
      boxExtents=<10 15 15>
      constraintSetup {angularLimitsMax=<0 0 50>, angularLimitsMin=<0 -75 -50>, angularTarget=<0 0 0>}
      localJointOffset=<1 0 0>
     }]
    relativeSpaceBone {boneName=pack}
   }
   n25 = Anim Dynamics - Bone: dyn_strand_14!(ComponentPose=@n24, Alpha=1) -> (Pose) {
    angularSpringConstant=0.999
    bAngularSpring=true
    bUseGravityOverride=true
    bUseSphericalLimits=true
    simulationSpace=BoneRelative
    alphaBoolBlend {blendOption=Linear}
    boundBone {boneName=dyn_strand_14}
    componentAppliedLinearAccClamp=<1300 1300 1300>
    componentLinearAccScale=<0.5 0.5 1>
    componentLinearVelScale=<0.5 0.5 1>
    gravityOverride {z=-500}
    physicsBodyDefinitions [{
      boundBone {boneName=dyn_strand_14}
      boxExtents=<5 10 10>
      constraintSetup {angularLimitsMax=<5 60 10>, angularLimitsMin=<-5 -60 -10>, angularTarget=<0 0 0>}
      localJointOffset=<1 0 0>
     }]
    relativeSpaceBone {boneName=pack}
    sphericalLimits [{limitRadius=0, drivingBone {boneName=None}, sphereLocalOffset=<0 0 0>}]
   }
   n26 = Anim Dynamics - Bone: dyn_strand_15!(ComponentPose=@n25, Alpha=1) -> (Pose) {
    angularSpringConstant=0.999
    bAngularSpring=true
    bUseGravityOverride=true
    bUseSphericalLimits=true
    simulationSpace=BoneRelative
    alphaBoolBlend {blendOption=Linear}
    boundBone {boneName=dyn_strand_15}
    componentAppliedLinearAccClamp=<1300 1300 1300>
    componentLinearAccScale=<0.5 0.5 1>
    componentLinearVelScale=<0.5 0.5 1>
    gravityOverride {z=-500}
    physicsBodyDefinitions [{
      boundBone {boneName=dyn_strand_15}
      boxExtents=<10 15 15>
      constraintSetup {angularLimitsMax=<5 60 30>, angularLimitsMin=<-5 -60 -30>, angularTarget=<0 0 0>}
      localJointOffset=<1 0 0>
     }]
    relativeSpaceBone {boneName=pack}
   }
   n27 = Component To Local(ComponentPose=@n26)
   n28 = Get bIsInFrontEnd()
   n29 = "Transform (Modify) Bone - Bone: pack"!(
    ComponentPose=Local To Component(LocalPose=Blend Poses by bool(BlendPose_0=@n6, BlendPose_1=@n27, bActiveValue=@n28) {blendType=Linear})
    bAlphaBoolEnabled:bool=Get not own character() -> (not own character:bool)
    Translation="0,0.8,0"
    Rotation="0,0,0"
    Scale
   ) -> (Pose) {
    alphaInputType=Bool
    rotationMode=BMM_Additive
    rotationSpace=BCS_BoneSpace
    translationMode=BMM_Additive
    translationSpace=BCS_BoneSpace
    alphaBoolBlend {blendOption=Linear}
    boneToModify {boneName=pack}
    translation {y=0.8}
   }
   Output Pose(Result=Component To Local(ComponentPose=@n29)) {layerGroup=None}
   Save cached pose 'InputPose'(Pose=Input Pose()) {cacheName=InputPose}
  ]}
 EventGraph {nodes [
   n1 = Event BlueprintUpdateAnimation()
   > n2 = Set FPSAlpha(FPSAlpha:double=MapRangeClamped(Value=@n1.DeltaTimeX, InRangeA=0.0667, InRangeB=0.0167, OutRangeA=0.3)) -> (Output_Get:double)
   > Set FPSAlpha FE(FPSAlpha FE:double=float * float(A=@n2, B=0.6) {is=title K2Node_PromotableOperator}) -> (Output_Get:double)
   Event OnCharacterCustomizationLoaded()
   > n3 = Cast To FortPlayerPawn(Object=TryGetPawnOwner())
   n4 = GetDisplayName(Object=GetSkeletalMeshForPartType(self=@n3))
   > Set not own character(not own character:bool=NOT Boolean(A=MatchesWildcard(SourceString=@n4, Wildcard=*dollenthusiast*))) -> (Output_Get:bool)
  ]}
]
variables [
 FPSAlpha {type=number}
 FPSAlpha FE {type=number}
 not own character {type=boolean}
]
